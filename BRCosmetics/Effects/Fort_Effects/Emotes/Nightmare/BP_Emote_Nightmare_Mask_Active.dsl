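; Mask actor for the Nightmare emote that fades a glowing mask in on its skeletal mesh.
; Construction creates a dynamic material instance of the emote mask material, which BeginPlay applies to slot 0.
; A timeline drives Visibility and MaskEmissiveIntensity; its update also feeds a 1.35-second delay that swaps in the standard Nightmare face-accessory material and activates the Niagara effect.
; Tick and overlap events are empty.
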(event Collision|EventActorBeginOverlap
  (params (OtherActor :type Actor%20Object%20Reference)))

(event EventTick
  (params (DeltaSeconds :type Float%20%28single-precision%29)))

(event EventBeginPlay
  (bind _skeletalmesh (Variables|Default|GetSkeletalMesh))
  (|CallParentFunction)
  (Rendering|Material|SetMaterial _skeletalmesh 0 (Variables|Default|GetMID))
  (bind (_direction _time_emissive _time_visibility) (Timeline|Timeline_in PlayFromStart)
    (:Update
      (Rendering|Material|SetScalarParameterValueonMaterials _skeletalmesh "Visibility" (Math|Float|Lerp 0.0 1.0 _time_visibility))
      (Rendering|Material|SetScalarParameterValueonMaterials _skeletalmesh "MaskEmissiveIntensity" (Math|Float|Lerp 0.0 1.0 _time_emissive))
      (Utilities|FlowControl|Delay 1.35)
      (Rendering|Material|SetMaterial (Variables|Default|GetSkeletalMesh) 0 "/BRCosmetics/Characters/Player/Male/Medium/Bodies/M_MED_Nightmare/Materials/M_MED_Nightmare_FaceAcc.M_MED_Nightmare_FaceAcc")
      (Components|Activation|SetActive (Variables|Default|GetNS_Emote_Nightmare_Active) true false))))

(fn ConstructionScript
  (|CallParentFunction)
  (bind _returnvalue (Rendering|Material|CreateDynamicMaterialInstance "/BRCosmetics/Effects/Fort_Effects/Emotes/Nightmare/MI_Emote_Nightmare_Mask.MI_Emote_Nightmare_Mask" "Nightmare_Emote_MID" "None" :!class /Script/Engine.KismetMaterialLibrary))
  (Variables|Default|SetMID _returnvalue))
