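; Blueprint for the mask deactivation effect of the Nightmare emote, swapping materials on a skeletal mesh and firing a particle component.
; The construction script creates a dynamic instance of the emote mask material.
; On BeginPlay, delayed sequence branches briefly apply it at zero visibility, restore the face accessory material, and activate the deactivate effect component.
; The last-timed branch reapplies the instance while a timeline drives its emissive intensity and visibility.
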(event Collision|EventActorBeginOverlap
  (params (OtherActor :type Actor%20Object%20Reference)))

(event EventTick
  (params (DeltaSeconds :type Float%20%28single-precision%29)))

(event EventBeginPlay
  (bind _skeletalmesh (Variables|Default|GetSkeletalMesh))
  (bind _skeletalmesh_2 (Variables|Default|GetSkeletalMesh))
  (|CallParentFunction)
  (Utilities|FlowControl|Sequence
    (:then_0
      (Rendering|Material|SetMaterial _skeletalmesh_2 0 (Variables|Default|GetMID))
      (Rendering|Material|SetScalarParameterValueonMaterials _skeletalmesh_2 "Visibility" 0.0)
      (Utilities|FlowControl|Delay 0.1)
      (Rendering|Material|SetMaterial _skeletalmesh_2 0 "/BRCosmetics/Characters/Player/Male/Medium/Bodies/M_MED_Nightmare/Materials/M_MED_Nightmare_FaceAcc.M_MED_Nightmare_FaceAcc"))
    (:then_1
      (Utilities|FlowControl|Delay 1.3)
      (Rendering|Material|SetMaterial (Variables|Default|GetSkeletalMesh) 0 (Variables|Default|GetMID))
      (bind (_direction _time_emissive _time_visibility) (Timeline|Timeline_out Play)
        (:Update
          (Rendering|Material|SetScalarParameterValueonMaterials _skeletalmesh "MaskEmissiveIntensity" (Math|Float|Lerp 0.0 1.0 _time_emissive))
          (Rendering|Material|SetScalarParameterValueonMaterials _skeletalmesh "Visibility" (Math|Float|Lerp 0.0 1.0 _time_visibility)))))
    (:then_2
      (Utilities|FlowControl|Delay 0.7)
      (Components|Activation|SetActive (Variables|Default|GetNS_Emote_Nightmare_Deactivate) true false))))

(fn ConstructionScript
  (|CallParentFunction)
  (bind _returnvalue (Rendering|Material|CreateDynamicMaterialInstance "/BRCosmetics/Effects/Fort_Effects/Emotes/Nightmare/MI_Emote_Nightmare_Mask.MI_Emote_Nightmare_Mask" "Nightmare_Emote_MID" "None" :!class /Script/Engine.KismetMaterialLibrary))
  (Variables|Default|SetMID _returnvalue))
