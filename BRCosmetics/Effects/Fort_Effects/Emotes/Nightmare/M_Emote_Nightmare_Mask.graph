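# Additive unlit material for the Nightmare emote mask on skeletal meshes, producing a shimmering chromatic aberration look.
# Blended normals, scrolling screen-space swirl noise and a time-oscillating shift, dithered on deferred, bend the reflection vector that samples the aberration texture.
# Quantized Fresnel and timestamp-driven transitions blend that result with an emissive color, plus an expire tint.
# A separate branch computes refraction from transition, visibility and camera distance, with dithering on deferred.
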Material /BRCosmetics/Effects/Fort_Effects/Emotes/Nightmare/M_Emote_Nightmare_Mask
graph {
 properties {
  bEnableResponsiveAA=true
  bUsedWithMeshDeformer=true
  bUsedWithSkeletalMesh=true
  blendMode=BLEND_Additive
  pixelDepthOffsetMode=PDOM_Legacy
  refractionMethod=RM_PixelNormalOffset
  shadingModel=MSM_Unlit
  usedShadingModels=Unlit
 }
 nodes [
  n1 = VectorParameter() {parameterName=EmissiveColor, defaultValue=<rgba 0.6282 0.9926 5 1>}
  n2 = ScalarParameter() {defaultValue=0.5, parameterName=TransitionLength}
  n3 = MaterialFunctionCall(
   TimestampA=ScalarParameter() {parameterName=TimeStampA}
   TransitionLength_In=@n2
   TimestampB=ScalarParameter() {defaultValue=10000000000, parameterName=TimeStampB}
   TransitionLength_Out=@n2
   Transition In Only?=StaticBool()
  ) {materialFunction=&MF_Timestamp_InOut_01}
  n4 = OneMinus(None=@n3.Linear)
  n5 = ScalarParameter() {parameterName=MaskEmissiveIntensity}
  n6 = LinearInterpolate(
   A=ScalarParameter() {defaultValue=1, parameterName=Visibility}
   B=ScalarParameter() {defaultValue=1.3, parameterName=ExpireVisibility}
   Alpha=@n5
  )
  n7 = Multiply(A=VectorParameter() {parameterName=ChromaShiftRandom, defaultValue=<rgba 0.375 -0.4238 0 0>}.RGB, B=Sine(None=Time()) {period=8.5})
  n8 = Multiply(A=MaterialFunctionCall() {materialFunction=&DitherTemporalAA}.Result, B=@n7)
  n9 = TextureSampleParameter2D() {parameterName=Normals, samplerType=SAMPLERTYPE_Normal, texture=/Engine/EngineMaterials/FlatNormal}
  n10 = VectorParameter() {parameterName=NormalIntensity, defaultValue=<rgba 0.5 0.5 1 0>}
  n11 = ScreenPosition()
  n12 = TextureObject() {samplerType=SAMPLERTYPE_Normal, texture=/Game/Global/Textures/Default/Noise/T_SwirlNoise_N}
  n13 = TextureSample(UVs=Panner(Coordinate=Multiply(A=@n11.ViewportUV) {constB=5}) {speedX=0.5, speedY=-1}, Tex=@n12) {samplerType=SAMPLERTYPE_Normal, texture=&T_SwirlNoise_N}
  n14 = VectorParameter() {parameterName=DetailNormalIntensity, defaultValue=<rgba 0.5 0.5 1 0>}
  n15 = Transform(None=MaterialFunctionCall(BaseNormal=Multiply(A=@n9.RGB, B=@n10.RGB), AdditionalNormal=Multiply(A=@n13.RGB, B=@n14.RGB)) {materialFunction=&BlendAngleCorrectedNormals}.Result)
  n16 = Normalize(VectorInput=Add(A=ShadingPathSwitch(Default=@n7, Deferred=@n8), B=@n15))
  n17 = TextureSample(UVs=ComponentMask(None=ReflectionVectorWS(CustomWorldNormal=@n16)) {g=true, r=true}) {
   samplerType=SAMPLERTYPE_LinearColor
   texture=/BRCosmetics/Effects/Fort_Effects/Emotes/Nightmare/Material/T_Emote_Nightmare_Mask_ChromaticAbberation
  }
  n18 = ScalarParameter() {defaultValue=-0.5, parameterName=Fraction}
  n19 = ScalarParameter() {defaultValue=2, parameterName=Desaturation}
  n20 = MaterialFunctionCall(
   Exponent=LinearInterpolate(
    A=ScalarParameter() {defaultValue=3, parameterName=FresnelExponentStationary}
    B=ScalarParameter() {defaultValue=0.4, parameterName=FresnelExponentMoving}
    Alpha=@n6
   )
   Vertex / Pixel Normal=@n15
  ) {materialFunction=&MF_Fresnel_FX}
  n21 = ScalarParameter() {defaultValue=2, parameterName=QuantizeSteps}
  n22 = Divide(A=Floor(None=Multiply(A=@n20.Fresnel Inverted, B=@n21)), B=@n21)
  n23 = Multiply(
   A=LinearInterpolate(
    A=ScalarParameter() {defaultValue=0.01, parameterName=ChromaticScaleStationary}
    B=ScalarParameter() {defaultValue=0.065, parameterName=ChromaticScaleMoving}
    Alpha=@n6
   )
   B=Desaturation(None=@n17.RGB, Fraction=LinearInterpolate(A=@n18, B=@n19, Alpha=@n22) {constB=0})
  )
  n24 = Multiply(A=@n20.Fresnel Inverted, B=ScalarParameter() {defaultValue=1, parameterName=MaskFreIntensity}) {constB=10}
  n25 = Add(A=OneMinus(None=@n22), B=ConstantBiasScale(None=@n3.Linear) {bias=-0.5, scale=2})
  Add(
   A=LinearInterpolate(A=Multiply(A=@n1.RGB, B=@n4), B=Multiply(A=@n23, B=@n24), Alpha=Saturate(None=@n25))
   B=Multiply(A=@n5, B=VectorParameter() {parameterName=ExpireColorAdd, defaultValue=<rgba 0.0024 0.0076 0.065 0>}.RGB)
  )
  ConstantBiasScale(None=Sine(None=Time()))
  n26 = ScalarParameter() {defaultValue=0.5, parameterName=ExpireTransitionLength}
  MaterialFunctionCall(
   TimestampA=ScalarParameter() {parameterName=ExpireTimeStampA}
   TransitionLength_In=@n26
   TimestampB=ScalarParameter() {parameterName=ExpireTimeStampB}
   TransitionLength_Out=@n26
   Transition In Only?=StaticBool()
  ) {materialFunction=&MF_Timestamp_InOut_01}
  MaterialFunctionCall(
   TextureObject=@n12
   TextureSize=VectorParameter() {parameterName=NoiseTextureSize, defaultValue=<rgba 220 220 220 0>}.RGB
   WorldPosition=VertexInterpolator(VS=PreSkinnedPosition()).PS
   World Space Normal=VertexInterpolator(VS=PreSkinnedNormal()).PS
  ) {materialFunction=&WorldAlignedTexture}
  ScalarParameter() {defaultValue=1, parameterName=DebugMovement}
  ScalarParameter() {defaultValue=1, parameterName=DebugTime}
  n27 = ActorPositionWS()
  n28 = CameraPositionWS()
  n29 = Saturate(None=Divide(A=Distance(A=@n27, B=@n28), B=ScalarParameter() {defaultValue=800, parameterName=RefractionDistanceDivisor}))
  n30 = LinearInterpolate(
   A=ScalarParameter() {defaultValue=1.2, parameterName=RefractionTransition}
   B=LinearInterpolate(
    A=ScalarParameter() {defaultValue=0.995, parameterName=RefractionStationary}
    B=ScalarParameter() {defaultValue=0.98, parameterName=RefractionMoving}
    Alpha=Multiply(A=@n6, B=@n29)
   )
   Alpha=@n3.Linear
  )
  n31 = Abs(None=Divide(A=Subtract(A=@n30)))
  n32 = Add(A=@n31)
  n33 = Subtract(B=@n31)
  n34 = Multiply(
   A=MaterialFunctionCall() {materialFunction=&DitherTemporalAA}.Result
   B=Multiply(A=ScalarParameter() {defaultValue=0.05, parameterName=DitherScale}, B=ScalarParameter() {defaultValue=1, parameterName=Visibility})
  )
  ShadingPathSwitch(Default=Constant() {r=1}, Deferred=Multiply(A=@n30, B=LinearInterpolate(A=@n32, B=@n33, Alpha=@n34) {constA=0.9, constB=1.1}))
  n35 = VertexNormalWS()
  n36 = ScalarParameter() {defaultValue=2, parameterName=QuantizeSteps}
  n37 = Multiply(A=@n35, B=@n36)
  ShadingPathSwitch(Default=@n35, Deferred=Divide(A=Floor(None=@n37), B=@n36))
 ]
}
