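# Additive, unlit Niagara sprite material in the Nightmare emote's cloak effects, which can act as a decal-like particle.
# A panning noise texture, randomized per particle via dynamic parameters, drives particle-tinted color.
# Alpha combines texture, particle alpha, sphere mask, depth fades and optional lifetime sine blinking.
# It also uses Fresnel on flattened scene-texture world normals (constant on mobile), with inverted depth fade limiting it to nearby surfaces in decal mode.
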Material /BRCosmetics/Effects/Fort_Effects/Emotes/Nightmare/Material/M_Nightmare_Cloak_ParticleDecal_Emote
graph {
 properties {bUsedWithNiagaraSprites=true, blendMode=BLEND_Additive, pixelDepthOffsetMode=PDOM_Legacy, shadingModel=MSM_Unlit, usedShadingModels=Unlit}
 nodes [
  n1 = TextureCoordinate()
  n2 = AppendVector(
   A=ScalarParameter() {defaultValue=1, group=Emissive, parameterName=NoiseTilingU, sortPriority=3}
   B=ScalarParameter() {defaultValue=1, group=Emissive, parameterName=NoiseTilingV, sortPriority=4}
  )
  n3 = DynamicParameter() {paramNames [FresnelExponent, PanRandom, Param3, Param4]}
  n4 = Multiply(A=Time(), B=LinearInterpolate(Alpha=@n3.PanRandom) {constA=0.5})
  n5 = Add(A=@n3.PanRandom, B=ScalarParameter() {defaultValue=1, group=Emissive, parameterName=PanAdd})
  n6 = ScalarParameter() {group=Emissive, parameterName=NoisePanU, sortPriority=5}
  n7 = ScalarParameter() {group=Emissive, parameterName=NoisePanV, sortPriority=6}
  n8 = TextureSampleParameter2D(UVs=Panner(Coordinate=Multiply(A=@n1, B=@n2), Time=Add(A=@n4, B=@n5), Speed=AppendVector(A=@n6, B=@n7)) {speedX=0.05, speedY=0.5}) {group=Texture, parameterName=NoiseTexture, texture=/Game/Effects/Fort_Effects/Textures/Electric/T_Elect_Out}
  n9 = Power(
   Base=StaticComponentMaskParameter(None=@n8.RGB) {defaultB=true, defaultG=true, defaultR=true, group=Texture, parameterName=TextureRGB}
   Exp=ScalarParameter() {defaultValue=1, group=Emissive, parameterName=TexturePower}
  )
  n10 = ScalarParameter() {defaultValue=1, group=Emissive, parameterName=TextureMult}
  n11 = Multiply(
   A=Desaturation(None=Multiply(A=@n9, B=@n10), Fraction=ScalarParameter() {group=Emissive, parameterName=TextureDesaturation})
   B=ParticleColor().RGB
  )
  Add(A=@n11)
  Constant3Vector()
  MaterialFunctionCall(Emissive=@n11) {materialFunction=&MF_FX_EmissiveAdjustments}
  n12 = SceneTexture() {sceneTextureId=PPI_WorldNormal}
  n13 = Fresnel(
   ExponentIn=Multiply(
    A=ScalarParameter() {defaultValue=1, group=Alpha, parameterName=FresnelExponent}
    B=DynamicParameter() {paramNames [FresnelExponent, PanRandom, Param3, Param4]}.FresnelExponent
   )
   Normal=MaterialFunctionCall(
    Normal=ComponentMask(None=@n12.Color) {b=true, g=true, r=true}
    Flatness=ScalarParameter() {defaultValue=0.2, group=Alpha, parameterName=FlattenNormal}
   ) {materialFunction=&FlattenNormal}.Result
  ) {baseReflectFraction=0, exponent=1}
  n14 = ShadingPathSwitch(Default=@n13, Deferred=@n13, Forward=Constant3Vector() {constant=<rgba 0 0 1 1>}, Mobile=@n13)
  n15 = Constant() {r=0.3}
  n16 = Multiply(
   A=Multiply(A=ParticleColor().A, B=ShadingPathSwitch(Default=@n14, Deferred=@n14, Mobile=@n15) {desc=Migrated from FeatureLevelSwitch})
   B=SphereMask(
    A=TextureCoordinate()
    B=Constant() {r=0.5}
    Radius=ScalarParameter() {defaultValue=0.5, group=Alpha, parameterName=SphereMaskRadius}
    Hardness=ScalarParameter() {group=Alpha, parameterName=SphereMaskHardness}
   ) {hardnessPercent=0}
  )
  n17 = DepthFade(FadeDistance=ScalarParameter() {defaultValue=35, group=Alpha, parameterName=VisibleDepthDistance}) {fadeDistanceDefault=80}
  n18 = DepthFade(
   Opacity=LinearInterpolate(B=OneMinus(None=@n17), Alpha=ScalarParameter() {group=Alpha, parameterName=Use Decal}) {constA=1}
   FadeDistance=ScalarParameter() {defaultValue=20, group=Alpha, parameterName=DepthFade}
  ) {fadeDistanceDefault=15}
  n19 = ParticleRelativeTime()
  n20 = ScalarParameter() {defaultValue=10, group=Blinking, parameterName=Time}
  n21 = Multiply(A=Sine(None=Multiply(A=@n19, B=@n20)), B=Constant() {r=0.5})
  n22 = Constant() {r=0.5}
  n23 = Clamp(None=Add(A=Add(A=@n21, B=@n22), B=ScalarParameter() {group=Blinking, parameterName=Force Max Sine}) {constA=1})
  n24 = ScalarParameter() {group=Blinking, parameterName=Use Blinking}
  Multiply(
   A=StaticComponentMaskParameter(None=@n8.RGBA) {defaultR=true, group=Texture, parameterName=TextureAlpha}
   B=Multiply(A=Multiply(A=@n16, B=@n18), B=LinearInterpolate(B=@n23, Alpha=@n24) {constA=1})
  )
 ]
}
