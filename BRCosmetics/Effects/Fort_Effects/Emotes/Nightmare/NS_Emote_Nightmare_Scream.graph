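# Niagara effect for the Nightmare Scream emote that sprays small, short-lived droplet sprites.
# The active Drop emitter spawns them continuously in a small sphere, pushes them out from an offset point, then applies drag and gravity while alpha fades in and out.
# A disabled Smoke emitter retains a directional, sub-UV animated puff with sprite and mesh renderers, a decaying spawn rate and dynamic material parameters.
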NiagaraSystem /BRCosmetics/Effects/Fort_Effects/Emotes/Nightmare/NS_Emote_Nightmare_Scream
dependencies {
 usedDataInterfaces [&NiagaraDataInterfaceCurve]
 usedDynamicInputs []
 usedModules [
  &SystemState
  &self:SystemSpawnScript.SystemScriptSource.SystemScriptGraph.NiagaraNodeAssignment_0.SetVariables_#1
 ]
}
emitters {
 Drop {
  emitterUpdateScript [
   &EmitterState()
   &SpawnRate(SpawnRate=15, Use Spawn Probability=false)
  ]
  particleSpawnScript [
   /Niagara/Modules/Spawn/Initialization/InitializeParticle(
    Color=<rgba 0.3229 0.3156 0.2792 5>
    Lifetime=&UniformRangedFloat(Minimum=0.125, Maximum=0.25)
    Module.Write Color=true
    Module.Write Lifetime=true
    Module.Write Mass=true
    Module.Write Material Random=true
    Module.Write Position=false
    Module.Write Scale=true
    Module.Write SpriteRotation=false
    Module.Write SpriteSize=true
    Position=&SimulationPosition()
    Sprite Size=&RandomRangeVector2D(Minimum=<1 3>, Maximum=<1.5 7>)
   )
   &SphereLocation(
    MaskSpawnGroup=false
    Offset Coordinate Space:ENiagaraCoordinateSpace=::World
    Override Seed=false
    Sphere Radius=&RandomRangeFloat(Minimum=3, Maximum=5)
    Surface Only=false
    UseNonUniformScale=false
    UseOffset=true
    UseSphereOrientation=false
   )
   &AddVelocityFromPoint(
    Module.Use Falloff Distance=false
    Origin Offset=<4 0 0>
    Velocity Strength=&RandomRangeFloat(Minimum=20, Maximum=300)
   )
  ]
  particleUpdateScript [
   &Drag(Drag=4, Use Linear Drag=true, Use Rotational Drag=false)
   &GravityForce(Gravity=<0 0 -200>)
   &SolveForcesAndVelocity(Clamp Velocity=false, Limit Acceleration=false)
   &ParticleState()
   &ScaleColor(
    Scale Alpha=&FloatFromCurve(FloatCurve=DataInterface() {propertyValues {Curve {keys [
         {arriveTangent=6.668, interpMode=RCIM_Cubic, leaveTangent=6.668, tangentMode=RCTM_User, time=0, value=0}
         {interpMode=RCIM_Cubic, tangentMode=RCTM_User, time=0.5, value=1}
         <at 1 0>
        ]}}})
    ScaleA=true
    ScaleRGB=true
   )
  ]
  rendererClasses [&NiagaraSpriteRendererProperties]
  renderers [{rendererClass=&NiagaraSpriteRendererProperties}]
 }
 Smoke {
  bEnabled=false
  emitterUpdateScript [
   &EmitterState()
   &SpawnRate(
    SpawnRate=&FloatFromCurve(
     FloatCurve=DataInterface() {propertyValues {Curve {keys [{arriveTangent=-1.7816, interpMode=RCIM_Cubic, leaveTangent=-1.7816, tangentMode=RCTM_User, time=0, value=1}, <at 1 0>]}}}
     CurveIndex=$Emitter.NormalizedLoopAge
     Scale Curve=100
    )
    Use Spawn Probability=false
   )
  ]
  particleSpawnScript [
   &InitializeParticle(
    Color=&MakeLinearColorFromVectorAndFloat(
     "Vector (RGB)"=&MakeVectorFromLinearColorRGB(Color=&RandomRangeLinearColor(Minimum=<rgba 0.5625 0.5057 0.2216 0.1>, Maximum=<rgba 0.6719 0.561 0.3392 0.3>))
     "Float (Alpha)"=&UniformRangedFloat(Minimum=0.2, Maximum=0.3)
    )
    Lifetime=&UniformRangedFloat(Minimum=0.15, Maximum=0.4)
    Mesh Scale=<3 3 3>
    Module.Write Color=true
    Module.Write Lifetime=true
    Module.Write Mass=true
    Module.Write Material Random=true
    Module.Write Position=false
    Module.Write Scale=true
    Module.Write SpriteRotation=true
    Module.Write SpriteSize=true
    Position=&SimulationPosition()
    Sprite Rotation=&NormalizedAngleToDegrees(NormalizedAngle=&UniformRangedFloat(Minimum=0, Maximum=1))
    Sprite Size=&Vector2DFromFloat(Value=&Multiply_Float(A=&UniformRangedFloat(Minimum=150, Maximum=70), B=0.1))
   )
   /SetVariables(Particles.InitSpriteRotRate=&UniformRangedFloat(Minimum=-3, Maximum=3)) {isSetParameters=true}
   &CameraOffset(Camera Offset Amount=&UniformRangedFloat(Minimum=5, Maximum=2))
   &AddVelocity(Velocity=&Multiply_Vector(A=<-150 0 0>, B=&RandomRangeVector(Minimum=<0.8 0.8 0.8>, Maximum=<1.2 1.2 1.2>)))
   /SetVariables(Particles.testPosition:Vector3f=$Output.SkeletalMeshLocation.SampledPosition) {isSetParameters=true}
  ]
  particleUpdateScript [
   &ParticleState()
   &SpriteRotationRate(Rotation Rate=&Multiply_Float(
     A=$Particles.InitSpriteRotRate
     B=&FloatFromCurve(FloatCurve=DataInterface() {propertyValues {Curve {keys [<at -0.0046 1.0113 RCIM_Cubic>,<at 1.0011 0.09 RCIM_Cubic>]}}})
    ))
   /Niagara/Modules/Update/SubUV/SubUVAnimation(Number Of Frames=64, SubUV Lookup Index=&Multiply_Float(A=$Particles.Age, B=0.66))
   &Drag(Drag=10, Use Linear Drag=true, Use Rotational Drag=false)
   &SolveForcesAndVelocity(Clamp Velocity=false, Limit Acceleration=false)
   &Color(
    Color=$Particles.Initial.Color
    Scale Alpha=&FloatFromCurve(FloatCurve=DataInterface() {propertyValues {Curve {keys [{arriveTangent=-0.8206, interpMode=RCIM_Cubic, leaveTangent=-0.8206, tangentMode=RCTM_User, time=0, value=0.1}, <at 1.0011 0.002 RCIM_Cubic>]}}})
    Scale Color=&VectorFromFloat(Value=2)
   )
   &ScaleSpriteSize(Scale Factor=&Vector2DFromFloat(Value=&FloatFromCurve(FloatCurve=DataInterface() {propertyValues {Curve {keys [
          {arriveTangent=10.4989, interpMode=RCIM_Cubic, leaveTangent=10.4712, tangentMode=RCTM_User, time=0, value=0.5}
          {arriveTangent=0.2629, interpMode=RCIM_Cubic, leaveTangent=3.8279, tangentMode=RCTM_User, time=1, value=3}
         ]}}})))
   &DynamicMaterialParameters(
    Index 0 Param 1=&RandomRangeFloat(Minimum=5, Maximum=1)
    Index 0 Param 2=1
    Index 0 Param 3=1
    Index 0 Param 4=1
    Module.Param0WriteEnabled=true
    Module.Param1WriteEnabled=true
    Module.Param2WriteEnabled=true
    Module.Param3WriteEnabled=true
   )
  ]
  rendererClasses [&NiagaraSpriteRendererProperties, &NiagaraMeshRendererProperties]
  renderers [
   {rendererClass=&NiagaraSpriteRendererProperties}
   {rendererClass=&NiagaraMeshRendererProperties, rendererIndex=1}
  ]
 }
}
summary {
 systemName=NS_Emote_Nightmare_Scream
 emitters [
  {emitterName=Drop, rendererClasses [&NiagaraSpriteRendererProperties]}
  {
   bEnabled=false
   emitterName=Smoke
   rendererClasses [&NiagaraSpriteRendererProperties, &NiagaraMeshRendererProperties]
  }
 ]
 userVariables []
}
