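# Niagara smoke effect for the Nightmare emote, made of short-lived, translucent blue-grey puffs from a single Smoke emitter.
# Spawn rate tapers over the emitter loop, and puffs launch directionally against strong drag.
# They rotate, grow and fade with SubUV animation and dynamic material parameters, feeding sprite and mesh renderers.
# Burst spawn, camera offset and skeletal mesh location are disabled, though the Skeletal Mesh user parameter remains.
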NiagaraSystem /BRCosmetics/Effects/Fort_Effects/Emotes/Nightmare/NS_Emote_Nightmare_Smoke
dependencies {usedDataInterfaces [&NiagaraDataInterfaceCurve], usedDynamicInputs [], usedModules [&SystemState]}
emitters {Smoke {
  emitterUpdateScript [
   &EmitterState()
   &SpawnBurst_Instantaneous(Spawn Count=&UniformRangedInt(Minimum=3, Maximum=5), Use Spawn Probability=false) {enabled=false}
   &SpawnRate(
    SpawnRate=&FloatFromCurve(
     FloatCurve=DataInterface() {propertyValues {Curve {keys [{arriveTangent=-1.7816, interpMode=RCIM_Cubic, leaveTangent=-1.7816, tangentMode=RCTM_User, time=0, value=1}, <at 1 0>]}}}
     CurveIndex=$Emitter.NormalizedLoopAge
     Scale Curve=250
    )
    Use Spawn Probability=false
   )
  ]
  particleSpawnScript [
   /Niagara/Modules/Spawn/Initialization/InitializeParticle(
    Color=&MakeLinearColorFromVectorAndFloat(
     "Vector (RGB)"=&MakeVectorFromLinearColorRGB(Color=&RandomRangeLinearColor(Minimum=<rgba 0.2216 0.3466 0.5625 1>, Maximum=<rgba 0.3392 0.4778 0.6719 1>))
     "Float (Alpha)"=&UniformRangedFloat(Minimum=0.2, Maximum=0.3)
    )
    Lifetime=&UniformRangedFloat(Minimum=0.15, Maximum=0.4)
    Mesh Scale=<3 3 3>
    Module.Write Color=true
    Module.Write Lifetime=true
    Module.Write Mass=true
    Module.Write Material Random=true
    Module.Write Position=false
    Module.Write Scale=true
    Module.Write SpriteRotation=true
    Module.Write SpriteSize=true
    Position=&SimulationPosition()
    Sprite Rotation=&NormalizedAngleToDegrees(NormalizedAngle=&UniformRangedFloat(Minimum=0, Maximum=1))
    Sprite Size=&Vector2DFromFloat(Value=&Multiply_Float(A=&UniformRangedFloat(Minimum=40, Maximum=70), B=0.1))
   )
   /SetVariables(Particles.InitSpriteRotRate=&UniformRangedFloat(Minimum=-3, Maximum=3)) {isSetParameters=true}
   &CameraOffset(Camera Offset Amount=&UniformRangedFloat(Minimum=5, Maximum=20)) {enabled=false}
   &SkeletalMeshLocation(Bone Sampling Mode=::NewEnumerator4, Inherit Velocity=false, MaskSpawnGroup=false, Skeletal Mesh=$User.Skeletal Mesh) {enabled=false}
   &AddVelocity(Velocity=&Multiply_Vector(A=<-350 75 150>, B=&RandomRangeVector(Minimum=<0.8 0.8 0.8>, Maximum=<1.2 1.2 1.2>)))
   /SetVariables(Particles.testPosition:Vector3f=$Output.SkeletalMeshLocation.SampledPosition) {isSetParameters=true}
  ]
  particleUpdateScript [
   &ParticleState()
   &SpriteRotationRate(Rotation Rate=&Multiply_Float(
     A=$Particles.InitSpriteRotRate
     B=&FloatFromCurve(FloatCurve=DataInterface() {propertyValues {Curve {keys [<at -0.0046 1.0113 RCIM_Cubic>,<at 1.0011 0.09 RCIM_Cubic>]}}})
    ))
   /Niagara/Modules/Update/SubUV/SubUVAnimation(Number Of Frames=64, SubUV Lookup Index=&Multiply_Float(A=$Particles.Age, B=0.66))
   &Drag(Drag=10, Use Linear Drag=true, Use Rotational Drag=false)
   &SolveForcesAndVelocity(Clamp Velocity=false, Limit Acceleration=false)
   &Color(
    Color=$Particles.Initial.Color
    Scale Alpha=&FloatFromCurve(FloatCurve=DataInterface() {propertyValues {Curve {keys [{arriveTangent=-0.8206, interpMode=RCIM_Cubic, leaveTangent=-0.8206, tangentMode=RCTM_User, time=0, value=0.5}, <at 1.0011 0.002 RCIM_Cubic>]}}})
    Scale Color=&VectorFromFloat(Value=2)
   )
   &ScaleSpriteSize(Scale Factor=&Vector2DFromFloat(Value=&FloatFromCurve(FloatCurve=DataInterface() {propertyValues {Curve {keys [
          {arriveTangent=4.007, interpMode=RCIM_Cubic, leaveTangent=3.9796, tangentMode=RCTM_User, time=-0.003, value=0}
          {arriveTangent=1.6558, interpMode=RCIM_Cubic, leaveTangent=5.2821, tangentMode=RCTM_User, time=1, value=3}
         ]}}})))
   &DynamicMaterialParameters(
    Index 0 Param 1=&RandomRangeFloat(Minimum=5, Maximum=1)
    Index 0 Param 2=1
    Index 0 Param 3=1
    Index 0 Param 4=1
    Module.Param0WriteEnabled=true
    Module.Param1WriteEnabled=true
    Module.Param2WriteEnabled=true
    Module.Param3WriteEnabled=true
   )
  ]
  rendererClasses [&NiagaraSpriteRendererProperties, &NiagaraMeshRendererProperties]
  renderers [
   {rendererClass=&NiagaraSpriteRendererProperties}
   {rendererClass=&NiagaraMeshRendererProperties, rendererIndex=1}
  ]
 }}
summary {
 systemName=NS_Emote_Nightmare_Smoke
 emitters [{emitterName=Smoke, rendererClasses [&NiagaraSpriteRendererProperties, &NiagaraMeshRendererProperties]}]
 userVariables [User.Skeletal Mesh {
   type=NiagaraDataInterfaceSkeletalMesh
   defaultValue {
    struct=NiagaraExt_VariableValue_DataInterface
    value {dataInterface=&self:NiagaraDataInterfaceSkeletalMesh_0}
   }
  }]
}
